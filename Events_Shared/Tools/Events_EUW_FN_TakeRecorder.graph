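# Editor Utility Widget that records the player character with Take Recorder and converts the take into a FortPlayerMannequin for Level Sequences.
# Buttons configure recording sources and properties, start or stop and prune the take, and review it.
# Conversion adds a spawnable, applies the CID, copies transform/animation tracks, and offsets Location.Z and Rotation.Z keys in delayed batches.
# Another button pastes the converted mannequin folder into the focused sequence.
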EditorUtilityWidgetBlueprint /Events_Shared/Tools/Events_EUW_FN_TakeRecorder
graphs [
 CanProcessChannel {nodes [
   n1 = CanProcessChannel!() -> (then, ChannelName:String) {is=title K2Node_FunctionEntry}
   > Return Node!(execute, Result:bool=OR Boolean(A=Contains(SearchIn=@n1, Substring=Location.Z), B=Contains(SearchIn=@n1, Substring=Rotation.Z)))
  ]}
 Copy Tracks {nodes [
   n1 = Copy Tracks!() -> (
    then
    Bindings:Array of Movie Scene Binding Proxy Structures
    Binding Names:Array of Strings
    Track Types:Array of Movie Scene Track Class References
   ) {is=title K2Node_FunctionEntry}
   > n2 = For Each Loop(Array:Array of Movie Scene Binding Proxy Structures=@n1) -> (Array Element:Movie Scene Binding Proxy Structure)
   n3 = Get Copied Tracks() -> (Copied Tracks:Array of Movie Scene Track Object References)
   n4 = CopyTracks!(
    execute=@n2.Completed
    self:Level Sequence Editor Subsystem Object Reference=Get LevelSequenceEditorSubsystem() -> (ReturnValue:Level Sequence Editor Subsystem Object Reference) {is=title K2Node_GetEditorSubsystem}
    Tracks:Array of Movie Scene Track Object References=@n3
   ) -> (then, ExportedText:String)
   > Return Node!(execute, Output Tracks:Array of Movie Scene Track Object References=@n3, Output Tracks Text:String=@n4)
   n5 = GetDisplayName!(execute:Exec=@n2.LoopBody, InBinding:"Movie Scene Binding Proxy Structure (by ref)"=@n2) -> (then:Exec, ReturnValue:Text)
   > Branch(Condition=Contains Item(TargetArray:Array of Strings=@n1.Binding Names, ItemToFind:"String (by ref)"="To String (Text)"(InText=@n5)))
   > n6 = GetTracks!(execute, InBinding:"Movie Scene Binding Proxy Structure (by ref)"=@n2) -> (then, ReturnValue:Array of Movie Scene Track Object References)
   > n7 = For Each Loop(Array:Array of Movie Scene Track Object References=@n6) -> (Array Element:Movie Scene Track Object Reference)
   > Branch(Condition=Contains Item(
     TargetArray:Array of Movie Scene Track Class References=Get Track Types() -> (Track Types:Array of Movie Scene Track Class References)
     ItemToFind:"Movie Scene Track Class Reference (by ref)"=Get Class(Object=@n7) -> (ReturnValue:Movie Scene Track Class Reference)
    ))
   > Add(TargetArray:Array of Movie Scene Track Object References=Get Copied Tracks(), NewItem:"Movie Scene Track Object Reference (by ref)"=@n7) {is=title K2Node_CallArrayFunction}
  ]}
 Copy/Paste Folders {nodes [
   n1 = Copy/Paste Folders!() -> (
    then
    Source Sequence:Level Sequence Object Reference
    Source Folders:Array of Movie Scene Folder Object References
    Source Bindings:Array of Movie Scene Binding Proxy Structures
    Target Sequence:Movie Scene Sequence Object Reference
   ) {is=title K2Node_FunctionEntry}
   n2 = Get LevelSequenceEditorSubsystem()
   > n3 = CopyFolders!(execute, self:Level Sequence Editor Subsystem Object Reference=@n2, Folders:Array of Movie Scene Folder Object References=@n1.Source Folders) -> (then, FoldersExportedText:String, ObjectsExportedText:String, TracksExportedText:String)
   > n4 = CopyBindings!(
    execute
    self:Level Sequence Editor Subsystem Object Reference=@n2
    Bindings:Array of Movie Scene Binding Proxy Structures=@n1.Source Bindings
   ) -> (then, ExportedText:String)
   > n5 = Execute Python Script!(execute, PythonScript="parent_folder = None") -> (then, ReturnValue, parent_folder:Movie Scene Folder Object Reference) {outputs [parent_folder]}
   > n6 = PasteFolders!(
    execute
    self:Level Sequence Editor Subsystem Object Reference=@n2
    TextToImport:String=@n3
    PasteFoldersParams:Movie Scene Paste Folders Params Structure=Make Movie Scene Paste Folders Params!(Sequence:Movie Scene Sequence Object Reference=@n1.Target Sequence, ParentFolder:Movie Scene Folder Object Reference=@n5.parent_folder) -> (MovieScenePasteFoldersParams:Movie Scene Paste Folders Params Structure) {is=title K2Node_MakeStruct}
   ) -> (then, OutFolders:Array of Movie Scene Folder Object References, ReturnValue:bool)
   > n7 = Execute Python Script!(execute, PythonScript="folder = []") -> (then, ReturnValue, bindings:Array of Movie Scene Binding Proxy Structures, folders:Array of Movie Scene Folder Object References) {outputs [bindings,folders]}
   n8 = "Get (a copy)"(Array:Array of Movie Scene Folder Object References=@n6) -> (Output:Movie Scene Folder Object Reference)
   > n9 = PasteBindings!(
    execute
    self:Level Sequence Editor Subsystem Object Reference=@n2
    TextToImport:String=@n4
    PasteBindingsParams:Movie Scene Paste Bindings Params Structure=Make Movie Scene Paste Bindings Params!(
     Bindings:Array of Movie Scene Binding Proxy Structures=@n7.bindings
     ParentFolder:Movie Scene Folder Object Reference=@n8
     Folders:Array of Movie Scene Folder Object References=@n7.folders
     bDuplicateExistingActors:bool=False
     PastedActors:Map of Names to Actor Object References
    ) -> (MovieScenePasteBindingsParams:Movie Scene Paste Bindings Params Structure) {is=title K2Node_MakeStruct}
   ) -> (then, OutObjectBindings:Array of Movie Scene Binding Proxy Structures, ReturnValue:bool)
   > Copy/Paste Tracks!(
    execute
    self
    Copied Bindings:Array of Movie Scene Binding Proxy Structures=Get Source Bindings() -> (Source Bindings:Array of Movie Scene Binding Proxy Structures)
    Copied Binding Names:Array of Strings=Get Copied Component Names() -> (Copied Component Names:Array of Strings)
    Copied Track Types:Array of Movie Scene Track Class References=Get Copied Track Types() -> (Copied Track Types:Array of Movie Scene Track Class References)
    Paste Sequence:Movie Scene Sequence Object Reference=Get Target Sequence() -> (Target Sequence:Movie Scene Sequence Object Reference)
    Paste Bindings:Array of Movie Scene Binding Proxy Structures=@n9
    Paste Parent Folder:Movie Scene Folder Object Reference=@n8
    Paste Folders:Array of Movie Scene Folder Object References=Make Array("[0]":Movie Scene Folder Object Reference) -> (Array:Array of Movie Scene Folder Object References)
   ) -> (then)
  ]}
 Copy/Paste Tracks {nodes [
   n1 = Copy/Paste Tracks!() -> (
    then
    Copied Bindings:Array of Movie Scene Binding Proxy Structures
    Copied Binding Names:Array of Strings
    Copied Track Types:Array of Movie Scene Track Class References
    Paste Sequence:Movie Scene Sequence Object Reference
    Paste Bindings:Array of Movie Scene Binding Proxy Structures
    Paste Parent Folder:Movie Scene Folder Object Reference
    Paste Folders:Array of Movie Scene Folder Object References
   ) {is=title K2Node_FunctionEntry}
   > n2 = Copy Tracks!(
    execute
    self
    Bindings:Array of Movie Scene Binding Proxy Structures=@n1
    Binding Names:Array of Strings=@n1.Copied Binding Names
    Track Types:Array of Movie Scene Track Class References=@n1.Copied Track Types
   ) -> (then, Output Tracks:Array of Movie Scene Track Object References, Output Tracks Text:String) {is=title K2Node_CallFunction}
   > PasteTracks!(
    execute
    self:Level Sequence Editor Subsystem Object Reference=Get LevelSequenceEditorSubsystem()
    TextToImport:String=@n2.Output Tracks Text
    PasteTracksParams_Sequence:Movie Scene Sequence Object Reference=@n1.Paste Sequence
    PasteTracksParams_Bindings:Array of Movie Scene Binding Proxy Structures=@n1.Paste Bindings
    PasteTracksParams_ParentFolder:Movie Scene Folder Object Reference=@n1.Paste Parent Folder
    PasteTracksParams_Folders:Array of Movie Scene Folder Object References=@n1.Paste Folders
   ) -> (then, OutTracks:Array of Movie Scene Track Object References, ReturnValue:bool)
  ]}
 EventGraph {nodes [
   "On Clicked (BTN_RecordPlayerCharacter)"() {is=title K2Node_ComponentBoundEvent}
   > OpenTakeRecorderPanel!(execute) -> (then, ReturnValue:Take Recorder Panel Object Reference)
   > Delay!(execute, WorldContextObject:Object Reference, Duration=0.5) -> (then)
   > n1 = GetGameWorld!(execute, self:Unreal Editor Subsystem Object Reference=Get UnrealEditorSubsystem()) -> (then, ReturnValue:World Object Reference)
   > n2 = Sequence!(execute) -> (then_0, then_1, then_2:Exec)
   n3 = GetSources!(self:Take Recorder Panel Object Reference=GetTakeRecorderPanel!() -> (ReturnValue:Take Recorder Panel Object Reference)) -> (ReturnValue:Take Recorder Sources Object Reference)
   > n4 = For Each Loop(Array:Array of Take Recorder Source Object References="Get Sources (Copy)"!(self:Take Recorder Sources Object Reference=@n3) -> (ReturnValue:Array of Take Recorder Source Object References) {is=title K2Node_CallFunction}) -> (Array Element:Take Recorder Source Object Reference)
   > RemoveSource!(execute, self:Take Recorder Sources Object Reference=@n3, InSource:Take Recorder Source Object Reference=@n4) -> (then)
   > "Set Mode (Recording w/ Level Sequence)"!(execute, self:Take Recorder Panel Object Reference=GetTakeRecorderPanel(), LevelSequenceAsset:Level Sequence Object Reference) -> (then) {is=title K2Node_CallFunction}
   n5 = GetTakeRecorderPanel()
   AddSourceForActor!(
    execute=@n2.then_1
    InActor:Actor Object Reference=GetPlayerCharacter!(WorldContextObject:Object Reference=@n1, PlayerIndex:int=0) -> (ReturnValue:Character Object Reference)
    InSources:Take Recorder Sources Object Reference=GetSources(self=@n5)
   ) -> (then, ReturnValue:Take Recorder Source Object Reference)
   > SetFrameRate!(execute, self:Take Recorder Panel Object Reference=@n5, InFrameRate_Numerator:int=60, InFrameRate_Denominator:int=1) -> (then)
   n6 = GetTakeMetaData!(self:Take Recorder Panel Object Reference=@n5) -> (ReturnValue:Take Meta Data Object Reference)
   n7 = Get Recording Name() -> (Recording Name:String)
   > SetSlate!(execute, self:Take Meta Data Object Reference=@n6, InSlate:String=@n7, bEmitChanged:bool=true) -> (then)
   > n8 = ComputeNextTakeNumber!(execute, Slate:String=@n7) -> (then, ReturnValue:int=0)
   > SetTakeNumber!(execute, self:Take Meta Data Object Reference=@n6, InTakeNumber:int=@n8, bEmitChanged:bool=true) -> (then)
   > SetFrameRate(self:Take Meta Data Object Reference=@n6, InFrameRate_Numerator=60, InFrameRate_Denominator=1)
   > n9 = GetDefaultParameters!(execute) -> (then, ReturnValue:Take Recorder Parameters Structure)
   n10 = Break Take Recorder Parameters!(TakeRecorderParameters:"Take Recorder Parameters Structure (by ref)"=@n9) -> (
    User:Take Recorder User Parameters Structure
    Project:Take Recorder Project Parameters Structure
    HitchProtection:Take Recorder Hitch Protection Parameters Structure
    TakeRecorderMode:ETakeRecorderMode Enum
    StartFrame:Frame Number Structure
    bOpenSequencer:bool=~
   ) {is=title K2Node_BreakStruct}
   > n11 = Set members in Take Recorder Project Parameters!(execute:Exec, StructRef:"Take Recorder Project Parameters Structure (by ref)"=@n10.Project, bRecordSourcesIntoSubSequences:bool=False) -> (then:Exec, StructOut:"Take Recorder Project Parameters Structure (by ref)") {is=title K2Node_SetFieldsInStruct}
   > SetDefaultParameters!(
    execute
    DefaultParameters_User:Take Recorder User Parameters Structure=@n10
    DefaultParameters_Project:Take Recorder Project Parameters Structure=@n11
    DefaultParameters_HitchProtection:Take Recorder Hitch Protection Parameters Structure=(bEnableHitchProtection=True,RegressionBufferSizeInSeconds=1,CustomTimestep="/Script/TimeManagement.CatchupFixedRateCustomTimeStep",MaxCatchupSeconds=8)
    DefaultParameters_TakeRecorderMode:ETakeRecorderMode Enum=@n10.TakeRecorderMode
    DefaultParameters_StartFrame:Frame Number Structure=@n10.StartFrame
    DefaultParameters_bOpenSequencer:bool=True
   ) -> (then)
   n12 = GetTakeRecorderPanel()
   n13 = For Each Loop(Exec=@n2.then_2, Array="Get Sources (Copy)"(self=GetSources(self=@n12)))
   > n14 = Cast To TakeRecorderActorSource!(execute, Object=@n13) -> (then, CastFailed, AsTake Recorder Actor Source:Take Recorder Actor Source Object Reference) {is=title K2Node_DynamicCast}
   > Set bReduceKeys!(execute, bReduceKeys:bool=false, self:Take Recorder Actor Source Object Reference=@n14) -> (then, Output_Get:bool)
   > Set bRecordParentHierarchy!(execute, bRecordParentHierarchy:bool=false, self:Take Recorder Actor Source Object Reference=@n14) -> (then, Output_Get:bool)
   > n15 = Sequence!(execute) -> (then_0, then_1)
   n16 = Get RecordedProperties!(self:Take Recorder Actor Source Object Reference=@n14) -> (RecordedProperties:Actor Recorder Property Map Object Reference)
   n17 = Get Properties!(self:Actor Recorder Property Map Object Reference=@n16) -> (Properties:Array of Actor Recorded Property Structures)
   > n18 = For Each Loop(Array:Array of Actor Recorded Property Structures=@n17) -> (Array Element:Actor Recorded Property Structure)
   n19 = "To String (Name)"(InName=Break Actor Recorded Property!(ActorRecordedProperty:"Actor Recorded Property Structure (by ref)"=@n18) -> (PropertyName:Name, bEnabled:bool=~) {is=title K2Node_BreakStruct})
   n20 = Make Array!("[0]":String=Spawn Track, "[1]":String=Transform Track) -> (Array:Array of Strings)
   > n21 = Set members in Actor Recorded Property!(
    execute:Exec
    StructRef:"Actor Recorded Property Structure (by ref)"=@n18
    bEnabled:bool="integer >= integer"(A=Find Item!(TargetArray=@n20, ItemToFind:"String (by ref)"=@n19) -> (ReturnValue))
   ) -> (then:Exec, StructOut:"Actor Recorded Property Structure (by ref)") {is=title K2Node_SetFieldsInStruct}
   > Set Array Elem!(
    execute
    TargetArray:Array of Actor Recorded Property Structures=@n17
    Index:int=@n18.Array Index
    Item:"Actor Recorded Property Structure (by ref)"=@n21
    bSizeToFit:bool=false
   ) -> (then)
   n22 = For Each Loop(
    Exec=@n15.then_1
    Array:Array of Actor Recorder Property Map Object References=Get Children!(self:Actor Recorder Property Map Object Reference=@n16) -> (Children:Array of Actor Recorder Property Map Object References)
   ) -> (Array Element:Actor Recorder Property Map Object Reference)
   > n23 = Branch(Condition="Equal Exactly (String)"(A=GetDisplayName!(Object:Object Reference=@n22) -> (ReturnValue:String), B=CollisionCylinder_0))
   > Set Component Properties!(
    execute
    self
    Component:Actor Recorder Property Map Object Reference=@n22
    Component Names:Array of Strings=Make Array!("[0]"=CharacterMesh0) -> (Array)
    Property Names:Array of Strings=Make Array("[0]"=Animation Track)
    Propagate to Children:bool=true
   ) -> (then) {is=title K2Node_CallFunction}
   Set Component Properties(execute=@n23.else, Component=@n22, Component Names=Make Array(), Property Names=Make Array(), Propagate to Children=true)
   "On Clicked (BTN_StopRecordingAndPIE)"() {is=title K2Node_ComponentBoundEvent}
   > StopRecording!(execute) -> (then)
   > EditorRequestEndPlay(self:Level Editor Subsystem Object Reference=Get LevelEditorSubsystem())
   > ReviewLastTake()
   > n24 = GetFocusedLevelSequence!(execute) -> (then, ReturnValue:Level Sequence Object Reference)
   > SetLockLevelSequence!(execute, bLock:bool=false) -> (then)
   > n25 = GetBindings!(execute, Sequence:Movie Scene Sequence Object Reference=@n24) -> (then, ReturnValue:Array of Movie Scene Binding Proxy Structures)
   > n26 = For Each Loop(Array:Array of Movie Scene Binding Proxy Structures=@n25) -> (Array Element:Movie Scene Binding Proxy Structure)
   > Branch(Condition=IsValid!(InBinding:"Movie Scene Binding Proxy Structure (by ref)"=@n26) -> (ReturnValue))
   > n27 = GetTracks(InBinding=@n26)
   > n28 = For Each Loop(Array:Array of Movie Scene Track Object References=@n27) -> (Array Element:Movie Scene Track Object Reference)
   > Is Valid(InputObject=@n28)
   > n29 = GetDisplayName!(execute:Exec, Track:Movie Scene Track Object Reference=@n28) -> (then:Exec, ReturnValue:Text)
   n30 = "To String (Text)"(InText=@n29)
   n31 = "Equal, Case Insensitive (String)"(A=@n30, B=Transform)
   n32 = OR Boolean(
    A=OR Boolean(A="Equal, Case Insensitive (String)"(A=@n30, B=CharacterMesh0), B="Equal, Case Insensitive (String)"(A=@n30, B=Animation))
    B="Equal, Case Insensitive (String)"(A=@n30, B=Spawned)
   )
   > n33 = Branch(Condition=OR Boolean(A=OR Boolean(A=@n32, B=@n31), B="Equal, Case Insensitive (String)"(A=@n30, B=RecordedAnimation)))
   RemoveTrack!(execute=@n33.else, InBinding:"Movie Scene Binding Proxy Structure (by ref)"=@n26, TrackToRemove:Movie Scene Track Object Reference=@n28) -> (then)
   "On Clicked (BTN_ConverPlayerToMannequin)"() {is=title K2Node_ComponentBoundEvent}
   > n34 = GetFocusedLevelSequence()
   > n35 = Is Valid(InputObject=@n34)
   PrintString(execute=@n35.Is Not Valid, InString="error: Couldn't convert object. Level Sequence is not focused!")
   SetLockLevelSequence(execute=@n35.Is Valid, bLock=false)
   > n36 = Sequence!(execute) -> (then_0, then_1, then_2:Exec, then_3:Exec, then_4:Exec)
   > n37 = AddSpawnableFromClass!(
    execute
    self:Level Sequence Editor Subsystem Object Reference=Get LevelSequenceEditorSubsystem()
    Sequence:Movie Scene Sequence Object Reference=@n34
    ClassToSpawn:Object Class Reference=/Script/FortniteGame.FortPlayerMannequin
   ) -> (then, ReturnValue:Movie Scene Binding Proxy Structure)
   > n38 = Is Valid(InputObject=Get CID() -> (CID:Athena Character Item Definition Object Reference))
   PrintString(execute=@n38.Is Not Valid, InString="warning: Couldn't add CID to FortPlayerMannequin. CID is null ")
   n39 = Branch(execute=@n38.Is Valid, Condition=IsValid(InBinding=@n37))
   > n40 = GetFocusedLevelSequence()
   > n41 = GetBindingID!(execute, Sequence:Movie Scene Sequence Object Reference=@n40, InBinding:"Movie Scene Binding Proxy Structure (by ref)"=@n37) -> (then, ReturnValue:Movie Scene Object Binding ID Structure)
   > n42 = For Each Loop(Array:Array of Object References=GetBoundObjects!(ObjectBinding:Movie Scene Object Binding ID Structure=@n41) -> (ReturnValue:Array of Object References)) -> (Array Element:Object Reference)
   > n43 = Cast To FortPlayerMannequin!(execute, Object=@n42) -> (then, CastFailed, AsFort Player Mannequin:Fort Player Mannequin Object Reference) {is=title K2Node_DynamicCast}
   > ResetAthenaCharacterDefAndVariants!(
    execute
    self:Fort Player Mannequin Object Reference=@n43
    NewAthenaCharacter:Athena Character Item Definition Object Reference=Get CID()
    InOverrideVariants:Array of Mcp Variant Channel Info Structures=Make Array("[0]":Mcp Variant Channel Info Structure) -> (Array:Array of Mcp Variant Channel Info Structures)
   ) -> (then)
   PrintString(execute=@n43.CastFailed, InString=warning: Converted Object is not a FortPlayerMannequin)
   PrintString(execute=@n39.else, InString=warning: Couldn't find the mannequin binding)
   n44 = GetBindings(execute=@n36.then_1, Sequence=@n34)
   > n45 = Execute Python Script!(execute, PythonScript="parent_folder = None\r\nfolders = []") -> (then, ReturnValue, parent_folder:Movie Scene Folder Object Reference, folders) {outputs [parent_folder, folders]}
   > Copy/Paste Tracks(
    Copied Bindings=@n44
    Copied Binding Names=Get Copied Component Names()
    Copied Track Types=Get Copied Track Types()
    Paste Sequence=@n34
    Paste Bindings=Make Array("[0]":Movie Scene Binding Proxy Structure=@n37) -> (Array:Array of Movie Scene Binding Proxy Structures)
    Paste Parent Folder=@n45.parent_folder
    Paste Folders=@n45.folders
   ) {is=title K2Node_CallFunction}
   n46 = AddRootFolderToSequence!(execute=@n36.then_2, Sequence:Movie Scene Sequence Object Reference=@n34, NewFolderName:String=Converted FN Mannequin) -> (then, ReturnValue:Movie Scene Folder Object Reference)
   > AddChildObjectBinding!(execute, Folder:Movie Scene Folder Object Reference=@n46, InObjectBinding:Movie Scene Binding Proxy Structure=@n37) -> (then, ReturnValue:bool)
   n47 = Get Binding Name Override() -> (Binding Name Override:String)
   > n48 = Branch(Condition=IsEmpty(InString=@n47))
   SetName!(execute=@n48.else, InBinding:"Movie Scene Binding Proxy Structure (by ref)"=@n37, InName:String=@n47) -> (then)
   n49 = Set FN Take(execute=@n36.then_3, FN Take:Level Sequence Object Reference=@n34) -> (Output_Get:Level Sequence Object Reference)
   > Branch(Condition=Get DeletePlayerPawn() -> (DeletePlayerPawn:bool))
   > n50 = GetBindings(Sequence=@n49)
   > n51 = For Each Loop(Array:Array of Movie Scene Binding Proxy Structures=@n50) -> (Array Element:Movie Scene Binding Proxy Structure)
   > n52 = GetName!(execute, InBinding:"Movie Scene Binding Proxy Structure (by ref)"=@n51) -> (then, ReturnValue:String)
   > Branch(Condition="Equal, Case Insensitive (String)"(A=@n52, B=PlayerPawn_Athena0))
   > Remove!(execute, InBinding:"Movie Scene Binding Proxy Structure (by ref)"=@n51) -> (then)
   > n53 = For Each Loop(Array:Array of Movie Scene Folder Object References=GetRootFoldersInSequence!(Sequence:Movie Scene Sequence Object Reference=@n49) -> (ReturnValue:Array of Movie Scene Folder Object References)) -> (Array Element:Movie Scene Folder Object Reference)
   n54 = GetFolderName!(Folder:Movie Scene Folder Object Reference=@n53) -> (ReturnValue:Name=None)
   > Branch(Condition="Equal, Case Insensitive (String)"(A="To String (Name)"(InName=@n54), B=Actors))
   > RemoveRootFolderFromSequence!(execute, Sequence:Movie Scene Sequence Object Reference=@n49, Folder:Movie Scene Folder Object Reference=@n53) -> (then)
   Set ChannelsIndex(execute=@n36.then_4, ChannelsIndex:int=0) -> (Output_Get:int=0)
   > n55 = FindTracksByExactType!(
    execute
    InBinding:"Movie Scene Binding Proxy Structure (by ref)"=@n37
    TrackType:Movie Scene Track Class Reference=/Script/MovieSceneTracks.MovieScene3DTransformTrack
   ) -> (then, ReturnValue:Array of Movie Scene 3DTransform Track Object References)
   > n56 = GetSections!(
    execute
    Track:Movie Scene Track Object Reference="Get (a copy)"(Array:Array of Movie Scene 3DTransform Track Object References=@n55) -> (Output:Movie Scene 3DTransform Track Object Reference)
   ) -> (then, ReturnValue:Array of Movie Scene Section Object References)
   > n57 = GetChannelsByType!(
    execute
    Section:Movie Scene Section Object Reference="Get (a copy)"(Array:Array of Movie Scene Section Object References=@n56) -> (Output:Movie Scene Section Object Reference)
    ChannelType:Movie Scene Scripting Channel Class Reference=/Script/SequencerScripting.MovieSceneScriptingDoubleChannel
   ) -> (then, ReturnValue:Array of Movie Scene Scripting Channel Object References)
   > Set ChannelsToProcess(ChannelsToProcess:Array of Movie Scene Scripting Channel Object References=@n57) -> (Output_Get:Array of Movie Scene Scripting Channel Object References)
   n58 = Length(TargetArray:Array of Movie Scene Scripting Channel Object References=@n57)
   > PrintString(InString="To String (Text)"(InText=Format Text!(Format="LOCTEXT(\"EUW_FN_TakeRecorder: Processing {length} channels\")", length:int=@n58) -> (Result)))
   n59 = Get ChannelsIndex() -> (ChannelsIndex:int=0)
   > While Loop!(execute:Exec, Condition:bool="integer < integer"(A=@n59, B=@n58)) -> (LoopBody:Exec, Completed:Exec) {is=title K2Node_MacroInstance}
   n60 = "Get (a copy)"(
    Array:Array of Movie Scene Scripting Channel Object References=Get ChannelsToProcess() -> (ChannelsToProcess:Array of Movie Scene Scripting Channel Object References)
    Dimension 1=@n59
   ) -> (Output:Movie Scene Scripting Channel Object Reference)
   n61 = "To String (Name)"(InName=Get ChannelName!(self:Movie Scene Scripting Channel Object Reference=@n60) -> (ChannelName:Name=None))
   > n62 = CanProcessChannel!(execute, self, ChannelName:String=@n61) -> (then, Result:bool) {is=title K2Node_CallFunction}
   > n63 = Branch(Condition=@n62)
   > n64 = Cast To MovieSceneScriptingDoubleChannel!(execute, Object=@n60) -> (then, CastFailed, AsMovie Scene Scripting Double Channel:Movie Scene Scripting Double Channel Object Reference) {is=title K2Node_DynamicCast}
   n65 = "Get Keys (Double)"!(self:Movie Scene Scripting Double Channel Object Reference=@n64) -> (ReturnValue:Array of Movie Scene Scripting Key Object References) {is=title K2Node_CallFunction}
   n66 = Length(TargetArray:Array of Movie Scene Scripting Key Object References=@n65)
   > PrintString(InString="To String (Text)"(InText=Format Text(Format="LOCTEXT(\"EUW_FN_TakeRecorder: Processing {length} transform keys\")", length=@n66)))
   > Set TransformKeyCount(TransformKeyCount:int=@n66) -> (Output_Get:int=0)
   > Set TransformKeyIndexCounter(TransformKeyIndexCounter:int=0) -> (Output_Get:int=0)
   > Set EditTransformIteration(EditTransformIteration:int=1) -> (Output_Get:int=0)
   > EditTransformKeysSafe!(execute, self, ChannelName:String=@n61, KeysToProcess:Array of Movie Scene Scripting Key Object References=@n65) -> (then)
   > Increment Int(Value=Get ChannelsIndex() -> (ChannelsIndex=0))
   Increment Int(" "=@n63.else, Value=Get ChannelsIndex() -> (ChannelsIndex=0))
   "On Clicked (BTN_StartRecording)"() {is=title K2Node_ComponentBoundEvent}
   > n67 = GetDefaultParameters()
   n68 = GetTakeRecorderPanel()
   > StartRecording!(
    execute
    LevelSequence:Level Sequence Object Reference=GetLevelSequence!(self:Take Recorder Panel Object Reference=@n68) -> (ReturnValue:Level Sequence Object Reference)
    Sources:Take Recorder Sources Object Reference=GetSources(self=@n68)
    MetaData:Take Meta Data Object Reference=GetTakeMetaData(self=@n68)
    Parameters:"Take Recorder Parameters Structure (by ref)"=@n67
   ) -> (then, ReturnValue:Take Recorder Object Reference)
   "On Clicked (BTN_AddRecordedFNMannequinTake)"() {is=title K2Node_ComponentBoundEvent}
   > n69 = For Each Loop(Array=GetRootFoldersInSequence(Sequence=Get FN Take() -> (FN Take:Level Sequence Object Reference)))
   n70 = GetFolderName(Folder=@n69) -> (ReturnValue=None)
   > Branch(Condition="Equal Exactly (String)"(A="To String (Name)"(InName=@n70), B=Converted FN Mannequin))
   > n71 = GetFocusedLevelSequence()
   > OpenLevelSequence!(execute, LevelSequence:Level Sequence Object Reference=Get FN Take()) -> (then, ReturnValue:bool)
   > SelectFolders!(
    execute
    Folders:Array of Movie Scene Folder Object References=Make Array("[0]":Movie Scene Folder Object Reference=@n69) -> (Array:Array of Movie Scene Folder Object References)
   ) -> (then)
   n72 = Get LevelSequenceEditorSubsystem()
   > n73 = CopyFolders(self=@n72, Folders=GetSelectedFolders!() -> (ReturnValue:Array of Movie Scene Folder Object References))
   > OpenLevelSequence(LevelSequence=@n71)
   > n74 = Execute Python Script!(execute, PythonScript="parent_folder = None") -> (then, ReturnValue, parent_folder) {outputs [parent_folder]}
   > PasteFolders(self=@n72, TextToImport=@n73, PasteFoldersParams=Make Movie Scene Paste Folders Params(Sequence=@n71, ParentFolder=@n74.parent_folder))
   "On Clicked (BTN_ReviewLastTake)"() {is=title K2Node_ComponentBoundEvent}
   > ReviewLastTake()
   n75 = EditTransformKeysSafe!() -> (OutputDelegate, then, ChannelName:String, KeysToProcess:Array of Movie Scene Scripting Key Object References) {is=title K2Node_CustomEvent}
   n76 = int * int(A=Get TransformKeyProcessSize() -> (TransformKeyProcessSize:int=0), B=Get EditTransformIteration() -> (EditTransformIteration:int=0))
   n77 = Get TransformKeyCount() -> (TransformKeyCount:int=0)
   > n78 = While Loop(Condition="integer < integer"(A=Get TransformKeyIndexCounter() -> (TransformKeyIndexCounter:int=0), B="Min (Integer)"(A=@n76, B=@n77) -> (ReturnValue=0)))
   n79 = Get TransformKeyIndexCounter() -> (TransformKeyIndexCounter=0)
   n80 = Get TransformKeyCount() -> (TransformKeyCount=0)
   PrintString(
    execute=@n78.Completed
    InString="To String (Text)"(InText=Format Text!(Format="LOCTEXT(\"EUW_FN_TakeRecorder: Processed {keys} / {total} transform keys\")", keys:int=@n79, total:int=@n80) -> (Result))
   )
   > n81 = Branch(Condition="integer >= integer"(A=Get TransformKeyIndexCounter() -> (TransformKeyIndexCounter=0), B=Get TransformKeyCount() -> (TransformKeyCount=0)))
   n82 = Increment Int(" "=@n81.else, Value=Get EditTransformIteration() -> (EditTransformIteration=0))
   > Set EditTransformIteration(EditTransformIteration=@n82) -> (Output_Get=0)
   > Delay(Duration=0.2)
   > EditTransformKeysSafe() {is=title K2Node_CallFunction}
   n83 = Branch(execute=@n78.LoopBody, Condition="Equal Exactly (String)"(A=@n75.ChannelName, B=Location.Z))
   n84 = "Get (a copy)"(
    Array:Array of Movie Scene Scripting Key Object References=@n75.KeysToProcess
    Dimension 1=Get TransformKeyIndexCounter() -> (TransformKeyIndexCounter=0)
   ) -> (Output:Movie Scene Scripting Key Object Reference)
   > n85 = Cast To MovieSceneScriptingDoubleKey!(execute, Object=@n84) -> (then, CastFailed, AsMovie Scene Scripting Double Key:Movie Scene Scripting Double Key Object Reference) {is=title K2Node_DynamicCast}
   > "Set Value (Double)"!(
    execute
    self:Movie Scene Scripting Double Key Object Reference=@n85
    InNewValue:double=float + float(
     A="Get Value (Double)"!(self:Movie Scene Scripting Double Key Object Reference=@n85) -> (ReturnValue:double) {is=title K2Node_CallFunction}
     B=-75
    )
   ) -> (then) {is=title K2Node_CallFunction}
   > n86 = Increment Int(Value=Get TransformKeyIndexCounter() -> (TransformKeyIndexCounter=0))
   > Set TransformKeyIndexCounter(TransformKeyIndexCounter=@n86) -> (Output_Get=0)
   n87 = Branch(execute=@n83.else, Condition="Equal Exactly (String)"(A=@n75.ChannelName, B=Rotation.Z))
   > n88 = Cast To MovieSceneScriptingDoubleKey(Object=@n84)
   > "Set Value (Double)"(self=@n88, InNewValue=float + float(A="Get Value (Double)"(self=@n88), B=-90))
   > n89 = Increment Int(Value=Get TransformKeyIndexCounter() -> (TransformKeyIndexCounter=0))
   > Set TransformKeyIndexCounter(TransformKeyIndexCounter=@n89) -> (Output_Get=0)
   n90 = Increment Int(" "=@n87.else, Value=Get TransformKeyIndexCounter() -> (TransformKeyIndexCounter=0))
   > Set TransformKeyIndexCounter(TransformKeyIndexCounter=@n90) -> (Output_Get=0)
   Event PreConstruct()
   > SetObject(
    self=[
     Get DetailsView_AddRecordedMannequinTake() -> (DetailsView_AddRecordedMannequinTake:Details View Object Reference)
     Get DetailsView_Recording() -> (DetailsView_Recording:Details View Object Reference)
     Get DetailsView_PostProcess() -> (DetailsView_PostProcess:Details View Object Reference)
    ]
    NewObject=Self-Reference()
   )
  ]}
 Record Component Properties {nodes [
   n1 = Record Component Properties!() -> (
    then
    Take Recorder Source:Take Recorder Actor Source Object Reference
    Component To Record:Array of Strings
    Property Names:Array of Strings
    Propagate to Children:bool=~
   ) {is=title K2Node_FunctionEntry}
   > n2 = For Each Loop(Array:Array of Actor Recorder Property Map Object References=Get Children(self=Get RecordedProperties(self=@n1))) -> (Array Element:Actor Recorder Property Map Object Reference)
   > Set Component Properties(Component=@n2, Component Names=@n1.Component To Record, Property Names=@n1.Property Names, Propagate to Children=@n1.Propagate to Children)
  ]}
 ReviewLastTake {nodes [
   ReviewLastTake() {is=title K2Node_FunctionEntry}
   n1 = GetTakeRecorderPanel()
   n2 = GetLastRecordedLevelSequence!(self:Take Recorder Panel Object Reference=@n1) -> (ReturnValue:Level Sequence Object Reference)
   > n3 = Is Valid(InputObject=@n2)
   PrintString(
    execute=@n3.Is Not Valid
    InString=error: Couldn't find Last Recorded Level Sequence
    bPrintToScreen=false
    TextColor=(R=1,G=0.0548,B=0.0748,A=1)
   )
   "Set Mode (Read-Only Level Sequence)"!(execute=@n3.Is Valid, self:Take Recorder Panel Object Reference=@n1, LevelSequenceAsset:Level Sequence Object Reference=@n2) -> (then) {is=title K2Node_CallFunction}
  ]}
 Set Component Properties {nodes [
   n1 = Set Component Properties!() -> (
    then
    Component:Actor Recorder Property Map Object Reference
    Component Names:Array of Strings
    Property Names:Array of Strings
    Propagate to Children:bool=~
   ) {is=title K2Node_FunctionEntry}
   > n2 = Sequence!(execute) -> (then_0, then_1, then_2, then_3)
   n3 = GetObjectName(Object=@n1)
   n4 = GetSubstring!(SourceString:String=@n3, StartIndex:int=0, Length:int=int - int(A=Len(S=@n3), B=2)) -> (ReturnValue:String)
   > Set Record(Record:bool="integer >= integer"(A=Find Item(TargetArray=@n1.Component Names, ItemToFind=@n4))) -> (Output_Get:bool)
   n5 = Get Properties(self=@n1)
   n6 = For Each Loop(Exec=@n2.then_1, Array:Array of Actor Recorded Property Structures=@n5) -> (Array Element:Actor Recorded Property Structure)
   Set Properties!(execute=@n6.Completed, Properties:Array of Actor Recorded Property Structures=@n5, self:Actor Recorder Property Map Object Reference=@n1) -> (then, Output_Get:Array of Actor Recorded Property Structures)
   n7 = Find Item(
    TargetArray=@n1.Property Names
    ItemToFind="To String (Name)"(InName=Break Actor Recorded Property(ActorRecordedProperty=@n6) -> (bEnabled=~))
   )
   n8 = Set members in Actor Recorded Property(execute=@n6.LoopBody, StructRef=@n6, bEnabled=AND Boolean(A=Get Record() -> (Record:bool), B="integer >= integer"(A=@n7)))
   > Set Array Elem(TargetArray=@n5, Index=@n6.Array Index, Item=@n8, bSizeToFit=false)
   Branch(execute=@n2.then_3, Condition=@n1.Propagate to Children)
   > n9 = For Each Loop(Array:Array of Actor Recorder Property Map Object References=Get Children(self=@n1)) -> (Array Element:Actor Recorder Property Map Object Reference)
   > Set Component Properties(Component=@n9, Component Names=@n1.Component Names, Property Names=@n1.Property Names, Propagate to Children=@n1.Propagate to Children) {is=title K2Node_CallFunction}
  ]}
]
variables [
 Binding Name Override {type=string, default=""}
 ChannelsIndex {type=integer}
 ChannelsToProcess {type="array<MovieSceneScriptingChannel>", default []}
 CID {type=AthenaCharacterItemDefinition, default=null}
 Copied Component Names {type="array<string>", default [PlayerPawn_Athena0, CharacterMesh0]}
 Copied Track Types {
  type="array<Class>"
  default [&MovieScene3DTransformTrack, /Script/MovieSceneTracks.MovieSceneSkeletalAnimationTrack]
 }
 DeletePlayerPawn {type=boolean}
 EditTransformIteration {type=integer}
 FN Take {type=LevelSequence, default=null}
 Recording Name {type=string, default=Scene_01}
 TransformKeyCount {type=integer}
 TransformKeyIndexCounter {type=integer}
 TransformKeyProcessSize {type=integer, default=2000}
]
widgets {
 info {namedSlotCount=4, rootWidgetClass=&ScrollBox, widgetCount=29}
 widgets [
  ScrollBox_0 {
   widgetClassPath=&ScrollBox
   widget=ScrollBox_0
   properties {
    widgetBarStyle {
     thickness=8
     draggedThumbImage {
      bIsDynamicallyLoaded=false
      drawAs=RoundedBox
      imageType=NoImage
      mirroring=NoMirror
      resourceName=None
      resourceObject=null
      tiling=NoTile
      imageSize=<0 0>
      margin=<ltrb 0 0 0 0>
      outlineSettings {
       bUseBrushTransparency=false
       roundingType=FixedRadius
       width=0
       color {colorUseRule=UseColor_Specified, specifiedColor=<rgba 0 0 0 0>}
       cornerRadii=<4 4 4 4>
      }
      tintColor {colorUseRule=UseColor_Specified, specifiedColor=<rgba 0.2159 0.2159 0.2159 1>}
      uVRegion {bIsValid=false, max=<0 0>, min=<0 0>}
     }
     horizontalBackgroundImage {
      bIsDynamicallyLoaded=false
      drawAs=NoDrawType
      imageType=NoImage
      mirroring=NoMirror
      resourceName=None
      resourceObject=null
      tiling=NoTile
      imageSize=<0 0>
      margin=<ltrb 0 0 0 0>
      outlineSettings {
       bUseBrushTransparency=false
       roundingType=HalfHeightRadius
       width=0
       color {colorUseRule=UseColor_Specified, specifiedColor=<rgba 0 0 0 0>}
       cornerRadii=<0 0 0 0>
      }
      tintColor {colorUseRule=UseColor_Specified, specifiedColor=<rgba 1 1 1 1>}
      uVRegion {bIsValid=false, max=<0 0>, min=<0 0>}
     }
     horizontalBottomSlotImage {
      bIsDynamicallyLoaded=false
      drawAs=NoDrawType
      imageType=NoImage
      mirroring=NoMirror
      resourceName=None
      resourceObject=null
      tiling=NoTile
      imageSize=<0 0>
      margin=<ltrb 0 0 0 0>
      outlineSettings {
       bUseBrushTransparency=false
       roundingType=HalfHeightRadius
       width=0
       color {colorUseRule=UseColor_Specified, specifiedColor=<rgba 0 0 0 0>}
       cornerRadii=<0 0 0 0>
      }
      tintColor {colorUseRule=UseColor_Specified, specifiedColor=<rgba 1 1 1 1>}
      uVRegion {bIsValid=false, max=<0 0>, min=<0 0>}
     }
     horizontalTopSlotImage {
      bIsDynamicallyLoaded=false
      drawAs=NoDrawType
      imageType=NoImage
      mirroring=NoMirror
      resourceName=None
      resourceObject=null
      tiling=NoTile
      imageSize=<0 0>
      margin=<ltrb 0 0 0 0>
      outlineSettings {
       bUseBrushTransparency=false
       roundingType=HalfHeightRadius
       width=0
       color {colorUseRule=UseColor_Specified, specifiedColor=<rgba 0 0 0 0>}
       cornerRadii=<0 0 0 0>
      }
      tintColor {colorUseRule=UseColor_Specified, specifiedColor=<rgba 1 1 1 1>}
      uVRegion {bIsValid=false, max=<0 0>, min=<0 0>}
     }
     hoveredThumbImage {
      bIsDynamicallyLoaded=false
      drawAs=RoundedBox
      imageType=NoImage
      mirroring=NoMirror
      resourceName=None
      resourceObject=null
      tiling=NoTile
      imageSize=<0 0>
      margin=<ltrb 0 0 0 0>
      outlineSettings {
       bUseBrushTransparency=false
       roundingType=FixedRadius
       width=0
       color {colorUseRule=UseColor_Specified, specifiedColor=<rgba 0 0 0 0>}
       cornerRadii=<4 4 4 4>
      }
      tintColor {colorUseRule=UseColor_Specified, specifiedColor=<rgba 0.2159 0.2159 0.2159 1>}
      uVRegion {bIsValid=false, max=<0 0>, min=<0 0>}
     }
     normalThumbImage {
      bIsDynamicallyLoaded=false
      drawAs=RoundedBox
      imageType=NoImage
      mirroring=NoMirror
      resourceName=None
      resourceObject=null
      tiling=NoTile
      imageSize=<0 0>
      margin=<ltrb 0 0 0 0>
      outlineSettings {
       bUseBrushTransparency=false
       roundingType=FixedRadius
       width=0
       color {colorUseRule=UseColor_Specified, specifiedColor=<rgba 0 0 0 0>}
       cornerRadii=<4 4 4 4>
      }
      tintColor {colorUseRule=UseColor_Specified, specifiedColor=<rgba 0.0953 0.0953 0.0953 1>}
      uVRegion {bIsValid=false, max=<0 0>, min=<0 0>}
     }
     verticalBackgroundImage {
      bIsDynamicallyLoaded=false
      drawAs=NoDrawType
      imageType=NoImage
      mirroring=NoMirror
      resourceName=None
      resourceObject=null
      tiling=NoTile
      imageSize=<0 0>
      margin=<ltrb 0 0 0 0>
      outlineSettings {
       bUseBrushTransparency=false
       roundingType=HalfHeightRadius
       width=0
       color {colorUseRule=UseColor_Specified, specifiedColor=<rgba 0 0 0 0>}
       cornerRadii=<0 0 0 0>
      }
      tintColor {colorUseRule=UseColor_Specified, specifiedColor=<rgba 1 1 1 1>}
      uVRegion {bIsValid=false, max=<0 0>, min=<0 0>}
     }
     verticalBottomSlotImage {
      bIsDynamicallyLoaded=false
      drawAs=NoDrawType
      imageType=NoImage
      mirroring=NoMirror
      resourceName=None
      resourceObject=null
      tiling=NoTile
      imageSize=<0 0>
      margin=<ltrb 0 0 0 0>
      outlineSettings {
       bUseBrushTransparency=false
       roundingType=HalfHeightRadius
       width=0
       color {colorUseRule=UseColor_Specified, specifiedColor=<rgba 0 0 0 0>}
       cornerRadii=<0 0 0 0>
      }
      tintColor {colorUseRule=UseColor_Specified, specifiedColor=<rgba 1 1 1 1>}
      uVRegion {bIsValid=false, max=<0 0>, min=<0 0>}
     }
     verticalTopSlotImage {
      bIsDynamicallyLoaded=false
      drawAs=NoDrawType
      imageType=NoImage
      mirroring=NoMirror
      resourceName=None
      resourceObject=null
      tiling=NoTile
      imageSize=<0 0>
      margin=<ltrb 0 0 0 0>
      outlineSettings {
       bUseBrushTransparency=false
       roundingType=HalfHeightRadius
       width=0
       color {colorUseRule=UseColor_Specified, specifiedColor=<rgba 0 0 0 0>}
       cornerRadii=<0 0 0 0>
      }
      tintColor {colorUseRule=UseColor_Specified, specifiedColor=<rgba 1 1 1 1>}
      uVRegion {bIsValid=false, max=<0 0>, min=<0 0>}
     }
    }
    widgetStyle {
     barThickness=8
     bottomShadowBrush {
      bIsDynamicallyLoaded=false
      drawAs=Box
      imageType=FullColor
      mirroring=NoMirror
      resourceName=../../../Engine/Content/Slate/Common/ScrollBoxShadowBottom.png
      resourceObject=null
      tiling=NoTile
      imageSize=<16 8>
      margin=<ltrb 0.5 0 0.5 1>
      outlineSettings {
       bUseBrushTransparency=false
       roundingType=HalfHeightRadius
       width=0
       color {colorUseRule=UseColor_Specified, specifiedColor=<rgba 0 0 0 0>}
       cornerRadii=<0 0 0 0>
      }
      tintColor {colorUseRule=UseColor_Specified, specifiedColor=<rgba 1 1 1 1>}
      uVRegion {bIsValid=false, max=<0 0>, min=<0 0>}
     }
     horizontalScrolledContentPadding=<ltrb 0 0 1 0>
     leftShadowBrush {
      bIsDynamicallyLoaded=false
      drawAs=Box
      imageType=FullColor
      mirroring=NoMirror
      resourceName=../../../Engine/Content/Slate/Common/ScrollBoxShadowLeft.png
      resourceObject=null
      tiling=NoTile
      imageSize=<8 16>
      margin=<ltrb 1 0.5 0 0.5>
      outlineSettings {
       bUseBrushTransparency=false
       roundingType=HalfHeightRadius
       width=0
       color {colorUseRule=UseColor_Specified, specifiedColor=<rgba 0 0 0 0>}
       cornerRadii=<0 0 0 0>
      }
      tintColor {colorUseRule=UseColor_Specified, specifiedColor=<rgba 1 1 1 1>}
      uVRegion {bIsValid=false, max=<0 0>, min=<0 0>}
     }
     rightShadowBrush {
      bIsDynamicallyLoaded=false
      drawAs=Box
      imageType=FullColor
      mirroring=NoMirror
      resourceName=../../../Engine/Content/Slate/Common/ScrollBoxShadowRight.png
      resourceObject=null
      tiling=NoTile
      imageSize=<8 16>
      margin=<ltrb 0 0.5 1 0.5>
      outlineSettings {
       bUseBrushTransparency=false
       roundingType=HalfHeightRadius
       width=0
       color {colorUseRule=UseColor_Specified, specifiedColor=<rgba 0 0 0 0>}
       cornerRadii=<0 0 0 0>
      }
      tintColor {colorUseRule=UseColor_Specified, specifiedColor=<rgba 1 1 1 1>}
      uVRegion {bIsValid=false, max=<0 0>, min=<0 0>}
     }
     topShadowBrush {
      bIsDynamicallyLoaded=false
      drawAs=Box
      imageType=FullColor
      mirroring=NoMirror
      resourceName=../../../Engine/Content/Slate/Common/ScrollBoxShadowTop.png
      resourceObject=null
      tiling=NoTile
      imageSize=<16 8>
      margin=<ltrb 0.5 1 0.5 0>
      outlineSettings {
       bUseBrushTransparency=false
       roundingType=HalfHeightRadius
       width=0
       color {colorUseRule=UseColor_Specified, specifiedColor=<rgba 0 0 0 0>}
       cornerRadii=<0 0 0 0>
      }
      tintColor {colorUseRule=UseColor_Specified, specifiedColor=<rgba 1 1 1 1>}
      uVRegion {bIsValid=false, max=<0 0>, min=<0 0>}
     }
     verticalScrolledContentPadding=<ltrb 0 0 0 1>
    }
   }
  }
  ExpandableArea_26 {
   widgetClassPath=/Script/UMG.ExpandableArea
   bIsVariable=true
   parent=ScrollBox_0
   widget=ExpandableArea_26
   properties {
    bIsExpanded=true
    borderBrush {
     bIsDynamicallyLoaded=false
     drawAs=RoundedBox
     imageType=NoImage
     mirroring=NoMirror
     resourceName=None
     resourceObject=null
     tiling=NoTile
     imageSize=<0 0>
     margin=<ltrb 0 0 0 0>
     outlineSettings {
      bUseBrushTransparency=false
      roundingType=FixedRadius
      width=0
      color {colorUseRule=UseColor_Specified, specifiedColor=<rgba 0 0 0 0>}
      cornerRadii=<4 4 4 4>
     }
     tintColor {colorUseRule=UseColor_Specified, specifiedColor=<rgba 0.0176 0.0176 0.0176 1>}
     uVRegion {bIsValid=false, max=<0 0>, min=<0 0>}
    }
    style {
     rolloutAnimationSeconds=0.1
     collapsedImage {
      bIsDynamicallyLoaded=false
      drawAs=Image
      imageType=Vector
      mirroring=NoMirror
      resourceName=../../../Engine/Content/Slate/Starship/Common/caret-right.svg
      resourceObject=null
      tiling=NoTile
      imageSize=<16 16>
      margin=<ltrb 0 0 0 0>
      outlineSettings {
       bUseBrushTransparency=false
       roundingType=HalfHeightRadius
       width=0
       color {colorUseRule=UseColor_Specified, specifiedColor=<rgba 0 0 0 0>}
       cornerRadii=<0 0 0 0>
      }
      tintColor {colorUseRule=UseColor_Specified, specifiedColor=<rgba 0.5271 0.5271 0.5271 1>}
      uVRegion {bIsValid=false, max=<0 0>, min=<0 0>}
     }
     expandedImage {
      bIsDynamicallyLoaded=false
      drawAs=Image
      imageType=Vector
      mirroring=NoMirror
      resourceName=../../../Engine/Content/Slate/Starship/Common/caret-down.svg
      resourceObject=null
      tiling=NoTile
      imageSize=<16 16>
      margin=<ltrb 0 0 0 0>
      outlineSettings {
       bUseBrushTransparency=false
       roundingType=HalfHeightRadius
       width=0
       color {colorUseRule=UseColor_Specified, specifiedColor=<rgba 0 0 0 0>}
       cornerRadii=<0 0 0 0>
      }
      tintColor {colorUseRule=UseColor_Specified, specifiedColor=<rgba 0.5271 0.5271 0.5271 1>}
      uVRegion {bIsValid=false, max=<0 0>, min=<0 0>}
     }
    }
   }
   slot {class=ScrollBoxSlot}
  }
  TextBlock_39 {
   widgetClassPath=&TextBlock
   namedSlotHost=ExpandableArea_26
   widget=TextBlock_39
   properties {
    text=Recording
    font {
     bForceMonospaced=false
     bMaterialIsStencil=false
     fontMaterial=null
     fontObject=&Roboto
     letterSpacing=0
     monospacedWidth=1
     size=20
     skewAmount=0
     typefaceFontName=Bold
     outlineSettings {
      bApplyOutlineToDropShadows=false
      bMiteredCorners=false
      bSeparateFillAlpha=false
      outlineMaterial=null
      outlineSize=0
      outlineColor=<rgba 0 0 0 1>
     }
    }
   }
  }
  VerticalBox_95 {widgetClassPath=&VerticalBox, namedSlotHost=ExpandableArea_26, widget=VerticalBox_95}
  DetailsView_Recording {
   widgetClassPath=&DetailsView
   parent=VerticalBox_95
   widget=DetailsView_Recording
   properties {categoriesToShow [Recording Settings]}
   slot {class=VerticalBoxSlot}
  }
  HorizontalBox_45 {widgetClassPath=&HorizontalBox, parent=VerticalBox_95, widget=HorizontalBox_45, slot {class=VerticalBoxSlot}}
  BTN_RecordPlayerCharacter {
   widgetClassPath=&Button
   parent=HorizontalBox_45
   widget=BTN_RecordPlayerCharacter
   properties {widgetStyle {
     clickedSlateSound {resourceObject=null}
     disabled {
      bIsDynamicallyLoaded=false
      drawAs=RoundedBox
      imageType=NoImage
      mirroring=NoMirror
      resourceName=None
      resourceObject=null
      tiling=NoTile
      imageSize=<0 0>
      margin=<ltrb 0 0 0 0>
      outlineSettings {
       bUseBrushTransparency=false
       roundingType=FixedRadius
       width=1
       color {colorUseRule=UseColor_ColorTable, specifiedColor=<rgba 0 0 0 0>}
       cornerRadii=<4 4 4 4>
      }
      tintColor {colorUseRule=UseColor_Specified, specifiedColor=<rgba 0.0395 0.0395 0.0395 1>}
      uVRegion {bIsValid=false, max=<0 0>, min=<0 0>}
     }
     disabledForeground {colorUseRule=UseColor_ColorTable, specifiedColor=<rgba 0 0 0 0>}
     hovered {
      bIsDynamicallyLoaded=false
      drawAs=RoundedBox
      imageType=NoImage
      mirroring=NoMirror
      resourceName=None
      resourceObject=null
      tiling=NoTile
      imageSize=<32 32>
      margin=<ltrb 0 0 0 0>
      outlineSettings {
       bUseBrushTransparency=false
       roundingType=FixedRadius
       width=1
       color {colorUseRule=UseColor_ColorTable, specifiedColor=<rgba 0 0 0 0>}
       cornerRadii=<4 4 4 4>
      }
      tintColor {colorUseRule=UseColor_Specified, specifiedColor=<rgba 0.0953 0.0953 0.0953 1>}
      uVRegion {bIsValid=false, max=<0 0>, min=<0 0>}
     }
     hoveredForeground {colorUseRule=UseColor_ColorTable, specifiedColor=<rgba 0 0 0 0>}
     hoveredSlateSound {resourceObject=null}
     normal {
      bIsDynamicallyLoaded=false
      drawAs=RoundedBox
      imageType=NoImage
      mirroring=NoMirror
      resourceName=None
      resourceObject=null
      tiling=NoTile
      imageSize=<32 32>
      margin=<ltrb 0 0 0 0>
      outlineSettings {
       bUseBrushTransparency=false
       roundingType=FixedRadius
       width=1
       color {colorUseRule=UseColor_ColorTable, specifiedColor=<rgba 0 0 0 0>}
       cornerRadii=<4 4 4 4>
      }
      tintColor {colorUseRule=UseColor_Specified, specifiedColor=<rgba 0.0395 0.0395 0.0395 1>}
      uVRegion {bIsValid=false, max=<0 0>, min=<0 0>}
     }
     normalForeground {colorUseRule=UseColor_ColorTable, specifiedColor=<rgba 0 0 0 0>}
     normalPadding=<ltrb 12 1.5 12 1.5>
     pressed {
      bIsDynamicallyLoaded=false
      drawAs=RoundedBox
      imageType=NoImage
      mirroring=NoMirror
      resourceName=None
      resourceObject=null
      tiling=NoTile
      imageSize=<32 32>
      margin=<ltrb 0 0 0 0>
      outlineSettings {
       bUseBrushTransparency=false
       roundingType=FixedRadius
       width=1
       color {colorUseRule=UseColor_ColorTable, specifiedColor=<rgba 0 0 0 0>}
       cornerRadii=<4 4 4 4>
      }
      tintColor {colorUseRule=UseColor_Specified, specifiedColor=<rgba 0.0284 0.0284 0.0284 1>}
      uVRegion {bIsValid=false, max=<0 0>, min=<0 0>}
     }
     pressedForeground {colorUseRule=UseColor_ColorTable, specifiedColor=<rgba 0 0 0 0>}
     pressedPadding=<ltrb 12 2.5 12 0.5>
     pressedSlateSound {resourceObject=null}
    }}
   slot {class=HorizontalBoxSlot, properties {padding=<ltrb 10 10 10 10>, size {sizeRule=Fill, value=1}}}
  }
  TextBlock_0 {
   widgetClassPath=&TextBlock
   parent=BTN_RecordPlayerCharacter
   widget=TextBlock_0
   properties {
    text=Setup Player Character
    font {
     bForceMonospaced=false
     bMaterialIsStencil=false
     fontMaterial=null
     fontObject=&Roboto
     letterSpacing=0
     monospacedWidth=1
     size=12
     skewAmount=0
     typefaceFontName=Regular
     outlineSettings {
      bApplyOutlineToDropShadows=false
      bMiteredCorners=false
      bSeparateFillAlpha=false
      outlineMaterial=null
      outlineSize=0
      outlineColor=<rgba 0 0 0 1>
     }
    }
   }
   slot {class=ButtonSlot}
  }
  BTN_StartRecording {
   widgetClassPath=&Button
   parent=HorizontalBox_45
   widget=BTN_StartRecording
   properties {widgetStyle {
     clickedSlateSound {resourceObject=null}
     disabled {
      bIsDynamicallyLoaded=false
      drawAs=RoundedBox
      imageType=NoImage
      mirroring=NoMirror
      resourceName=None
      resourceObject=null
      tiling=NoTile
      imageSize=<0 0>
      margin=<ltrb 0 0 0 0>
      outlineSettings {
       bUseBrushTransparency=false
       roundingType=FixedRadius
       width=1
       color {colorUseRule=UseColor_ColorTable, specifiedColor=<rgba 0 0 0 0>}
       cornerRadii=<4 4 4 4>
      }
      tintColor {colorUseRule=UseColor_Specified, specifiedColor=<rgba 0.0395 0.0395 0.0395 1>}
      uVRegion {bIsValid=false, max=<0 0>, min=<0 0>}
     }
     disabledForeground {colorUseRule=UseColor_ColorTable, specifiedColor=<rgba 0 0 0 0>}
     hovered {
      bIsDynamicallyLoaded=false
      drawAs=RoundedBox
      imageType=NoImage
      mirroring=NoMirror
      resourceName=None
      resourceObject=null
      tiling=NoTile
      imageSize=<32 32>
      margin=<ltrb 0 0 0 0>
      outlineSettings {
       bUseBrushTransparency=false
       roundingType=FixedRadius
       width=1
       color {colorUseRule=UseColor_ColorTable, specifiedColor=<rgba 0 0 0 0>}
       cornerRadii=<4 4 4 4>
      }
      tintColor {colorUseRule=UseColor_Specified, specifiedColor=<rgba 0.0953 0.0953 0.0953 1>}
      uVRegion {bIsValid=false, max=<0 0>, min=<0 0>}
     }
     hoveredForeground {colorUseRule=UseColor_ColorTable, specifiedColor=<rgba 0 0 0 0>}
     hoveredSlateSound {resourceObject=null}
     normal {
      bIsDynamicallyLoaded=false
      drawAs=RoundedBox
      imageType=NoImage
      mirroring=NoMirror
      resourceName=None
      resourceObject=null
      tiling=NoTile
      imageSize=<32 32>
      margin=<ltrb 0 0 0 0>
      outlineSettings {
       bUseBrushTransparency=false
       roundingType=FixedRadius
       width=1
       color {colorUseRule=UseColor_ColorTable, specifiedColor=<rgba 0 0 0 0>}
       cornerRadii=<4 4 4 4>
      }
      tintColor {colorUseRule=UseColor_Specified, specifiedColor=<rgba 0.0395 0.0395 0.0395 1>}
      uVRegion {bIsValid=false, max=<0 0>, min=<0 0>}
     }
     normalForeground {colorUseRule=UseColor_ColorTable, specifiedColor=<rgba 0 0 0 0>}
     normalPadding=<ltrb 12 1.5 12 1.5>
     pressed {
      bIsDynamicallyLoaded=false
      drawAs=RoundedBox
      imageType=NoImage
      mirroring=NoMirror
      resourceName=None
      resourceObject=null
      tiling=NoTile
      imageSize=<32 32>
      margin=<ltrb 0 0 0 0>
      outlineSettings {
       bUseBrushTransparency=false
       roundingType=FixedRadius
       width=1
       color {colorUseRule=UseColor_ColorTable, specifiedColor=<rgba 0 0 0 0>}
       cornerRadii=<4 4 4 4>
      }
      tintColor {colorUseRule=UseColor_Specified, specifiedColor=<rgba 0.0284 0.0284 0.0284 1>}
      uVRegion {bIsValid=false, max=<0 0>, min=<0 0>}
     }
     pressedForeground {colorUseRule=UseColor_ColorTable, specifiedColor=<rgba 0 0 0 0>}
     pressedPadding=<ltrb 12 2.5 12 0.5>
     pressedSlateSound {resourceObject=null}
    }}
   slot {class=HorizontalBoxSlot, properties {padding=<ltrb 10 10 10 10>, size {sizeRule=Fill, value=1}}}
  }
  TextBlock_3 {
   widgetClassPath=&TextBlock
   parent=BTN_StartRecording
   widget=TextBlock_3
   properties {
    text=Start Recording
    font {
     bForceMonospaced=false
     bMaterialIsStencil=false
     fontMaterial=null
     fontObject=&Roboto
     letterSpacing=0
     monospacedWidth=1
     size=12
     skewAmount=0
     typefaceFontName=Regular
     outlineSettings {
      bApplyOutlineToDropShadows=false
      bMiteredCorners=false
      bSeparateFillAlpha=false
      outlineMaterial=null
      outlineSize=0
      outlineColor=<rgba 0 0 0 1>
     }
    }
   }
   slot {class=ButtonSlot}
  }
  HorizontalBox {widgetClassPath=&HorizontalBox, parent=VerticalBox_95, widget=HorizontalBox, slot {class=VerticalBoxSlot}}
  BTN_StopRecordingAndPIE {
   widgetClassPath=&Button
   parent=HorizontalBox
   widget=BTN_StopRecordingAndPIE
   properties {widgetStyle {
     clickedSlateSound {resourceObject=null}
     disabled {
      bIsDynamicallyLoaded=false
      drawAs=RoundedBox
      imageType=NoImage
      mirroring=NoMirror
      resourceName=None
      resourceObject=null
      tiling=NoTile
      imageSize=<0 0>
      margin=<ltrb 0 0 0 0>
      outlineSettings {
       bUseBrushTransparency=false
       roundingType=FixedRadius
       width=1
       color {colorUseRule=UseColor_ColorTable, specifiedColor=<rgba 0 0 0 0>}
       cornerRadii=<4 4 4 4>
      }
      tintColor {colorUseRule=UseColor_Specified, specifiedColor=<rgba 0.0395 0.0395 0.0395 1>}
      uVRegion {bIsValid=false, max=<0 0>, min=<0 0>}
     }
     disabledForeground {colorUseRule=UseColor_ColorTable, specifiedColor=<rgba 0 0 0 0>}
     hovered {
      bIsDynamicallyLoaded=false
      drawAs=RoundedBox
      imageType=NoImage
      mirroring=NoMirror
      resourceName=None
      resourceObject=null
      tiling=NoTile
      imageSize=<32 32>
      margin=<ltrb 0 0 0 0>
      outlineSettings {
       bUseBrushTransparency=false
       roundingType=FixedRadius
       width=1
       color {colorUseRule=UseColor_ColorTable, specifiedColor=<rgba 0 0 0 0>}
       cornerRadii=<4 4 4 4>
      }
      tintColor {colorUseRule=UseColor_Specified, specifiedColor=<rgba 0.0953 0.0953 0.0953 1>}
      uVRegion {bIsValid=false, max=<0 0>, min=<0 0>}
     }
     hoveredForeground {colorUseRule=UseColor_ColorTable, specifiedColor=<rgba 0 0 0 0>}
     hoveredSlateSound {resourceObject=null}
     normal {
      bIsDynamicallyLoaded=false
      drawAs=RoundedBox
      imageType=NoImage
      mirroring=NoMirror
      resourceName=None
      resourceObject=null
      tiling=NoTile
      imageSize=<32 32>
      margin=<ltrb 0 0 0 0>
      outlineSettings {
       bUseBrushTransparency=false
       roundingType=FixedRadius
       width=1
       color {colorUseRule=UseColor_ColorTable, specifiedColor=<rgba 0 0 0 0>}
       cornerRadii=<4 4 4 4>
      }
      tintColor {colorUseRule=UseColor_Specified, specifiedColor=<rgba 0.0395 0.0395 0.0395 1>}
      uVRegion {bIsValid=false, max=<0 0>, min=<0 0>}
     }
     normalForeground {colorUseRule=UseColor_ColorTable, specifiedColor=<rgba 0 0 0 0>}
     normalPadding=<ltrb 12 1.5 12 1.5>
     pressed {
      bIsDynamicallyLoaded=false
      drawAs=RoundedBox
      imageType=NoImage
      mirroring=NoMirror
      resourceName=None
      resourceObject=null
      tiling=NoTile
      imageSize=<32 32>
      margin=<ltrb 0 0 0 0>
      outlineSettings {
       bUseBrushTransparency=false
       roundingType=FixedRadius
       width=1
       color {colorUseRule=UseColor_ColorTable, specifiedColor=<rgba 0 0 0 0>}
       cornerRadii=<4 4 4 4>
      }
      tintColor {colorUseRule=UseColor_Specified, specifiedColor=<rgba 0.0284 0.0284 0.0284 1>}
      uVRegion {bIsValid=false, max=<0 0>, min=<0 0>}
     }
     pressedForeground {colorUseRule=UseColor_ColorTable, specifiedColor=<rgba 0 0 0 0>}
     pressedPadding=<ltrb 12 2.5 12 0.5>
     pressedSlateSound {resourceObject=null}
    }}
   slot {class=HorizontalBoxSlot, properties {padding=<ltrb 10 10 10 10>, size {sizeRule=Fill, value=1}}}
  }
  TextBlock {
   widgetClassPath=&TextBlock
   parent=BTN_StopRecordingAndPIE
   widget=TextBlock
   properties {
    text=Stop Recording and PIE
    font {
     bForceMonospaced=false
     bMaterialIsStencil=false
     fontMaterial=null
     fontObject=&Roboto
     letterSpacing=0
     monospacedWidth=1
     size=12
     skewAmount=0
     typefaceFontName=Regular
     outlineSettings {
      bApplyOutlineToDropShadows=false
      bMiteredCorners=false
      bSeparateFillAlpha=false
      outlineMaterial=null
      outlineSize=0
      outlineColor=<rgba 0 0 0 1>
     }
    }
   }
   slot {class=ButtonSlot}
  }
  BTN_ReviewLastTake {
   widgetClassPath=&Button
   parent=HorizontalBox
   widget=BTN_ReviewLastTake
   properties {widgetStyle {
     clickedSlateSound {resourceObject=null}
     disabled {
      bIsDynamicallyLoaded=false
      drawAs=RoundedBox
      imageType=NoImage
      mirroring=NoMirror
      resourceName=None
      resourceObject=null
      tiling=NoTile
      imageSize=<0 0>
      margin=<ltrb 0 0 0 0>
      outlineSettings {
       bUseBrushTransparency=false
       roundingType=FixedRadius
       width=1
       color {colorUseRule=UseColor_ColorTable, specifiedColor=<rgba 0 0 0 0>}
       cornerRadii=<4 4 4 4>
      }
      tintColor {colorUseRule=UseColor_Specified, specifiedColor=<rgba 0.0395 0.0395 0.0395 1>}
      uVRegion {bIsValid=false, max=<0 0>, min=<0 0>}
     }
     disabledForeground {colorUseRule=UseColor_ColorTable, specifiedColor=<rgba 0 0 0 0>}
     hovered {
      bIsDynamicallyLoaded=false
      drawAs=RoundedBox
      imageType=NoImage
      mirroring=NoMirror
      resourceName=None
      resourceObject=null
      tiling=NoTile
      imageSize=<32 32>
      margin=<ltrb 0 0 0 0>
      outlineSettings {
       bUseBrushTransparency=false
       roundingType=FixedRadius
       width=1
       color {colorUseRule=UseColor_ColorTable, specifiedColor=<rgba 0 0 0 0>}
       cornerRadii=<4 4 4 4>
      }
      tintColor {colorUseRule=UseColor_Specified, specifiedColor=<rgba 0.0953 0.0953 0.0953 1>}
      uVRegion {bIsValid=false, max=<0 0>, min=<0 0>}
     }
     hoveredForeground {colorUseRule=UseColor_ColorTable, specifiedColor=<rgba 0 0 0 0>}
     hoveredSlateSound {resourceObject=null}
     normal {
      bIsDynamicallyLoaded=false
      drawAs=RoundedBox
      imageType=NoImage
      mirroring=NoMirror
      resourceName=None
      resourceObject=null
      tiling=NoTile
      imageSize=<32 32>
      margin=<ltrb 0 0 0 0>
      outlineSettings {
       bUseBrushTransparency=false
       roundingType=FixedRadius
       width=1
       color {colorUseRule=UseColor_ColorTable, specifiedColor=<rgba 0 0 0 0>}
       cornerRadii=<4 4 4 4>
      }
      tintColor {colorUseRule=UseColor_Specified, specifiedColor=<rgba 0.0395 0.0395 0.0395 1>}
      uVRegion {bIsValid=false, max=<0 0>, min=<0 0>}
     }
     normalForeground {colorUseRule=UseColor_ColorTable, specifiedColor=<rgba 0 0 0 0>}
     normalPadding=<ltrb 12 1.5 12 1.5>
     pressed {
      bIsDynamicallyLoaded=false
      drawAs=RoundedBox
      imageType=NoImage
      mirroring=NoMirror
      resourceName=None
      resourceObject=null
      tiling=NoTile
      imageSize=<32 32>
      margin=<ltrb 0 0 0 0>
      outlineSettings {
       bUseBrushTransparency=false
       roundingType=FixedRadius
       width=1
       color {colorUseRule=UseColor_ColorTable, specifiedColor=<rgba 0 0 0 0>}
       cornerRadii=<4 4 4 4>
      }
      tintColor {colorUseRule=UseColor_Specified, specifiedColor=<rgba 0.0284 0.0284 0.0284 1>}
      uVRegion {bIsValid=false, max=<0 0>, min=<0 0>}
     }
     pressedForeground {colorUseRule=UseColor_ColorTable, specifiedColor=<rgba 0 0 0 0>}
     pressedPadding=<ltrb 12 2.5 12 0.5>
     pressedSlateSound {resourceObject=null}
    }}
   slot {class=HorizontalBoxSlot, properties {padding=<ltrb 10 10 10 10>, size {sizeRule=Fill, value=1}}}
  }
  TextBlock_4 {
   widgetClassPath=&TextBlock
   parent=BTN_ReviewLastTake
   widget=TextBlock_4
   properties {
    text=Review Last Take
    font {
     bForceMonospaced=false
     bMaterialIsStencil=false
     fontMaterial=null
     fontObject=&Roboto
     letterSpacing=0
     monospacedWidth=1
     size=12
     skewAmount=0
     typefaceFontName=Regular
     outlineSettings {
      bApplyOutlineToDropShadows=false
      bMiteredCorners=false
      bSeparateFillAlpha=false
      outlineMaterial=null
      outlineSize=0
      outlineColor=<rgba 0 0 0 1>
     }
    }
   }
   slot {class=ButtonSlot}
  }
  ExpandableArea_52 {
   widgetClassPath=&ExpandableArea
   bIsVariable=true
   parent=ScrollBox_0
   widget=ExpandableArea_52
   properties {
    bIsExpanded=true
    borderBrush {
     bIsDynamicallyLoaded=false
     drawAs=RoundedBox
     imageType=NoImage
     mirroring=NoMirror
     resourceName=None
     resourceObject=null
     tiling=NoTile
     imageSize=<0 0>
     margin=<ltrb 0 0 0 0>
     outlineSettings {
      bUseBrushTransparency=false
      roundingType=FixedRadius
      width=0
      color {colorUseRule=UseColor_Specified, specifiedColor=<rgba 0 0 0 0>}
      cornerRadii=<4 4 4 4>
     }
     tintColor {colorUseRule=UseColor_Specified, specifiedColor=<rgba 0.0176 0.0176 0.0176 1>}
     uVRegion {bIsValid=false, max=<0 0>, min=<0 0>}
    }
    style {
     rolloutAnimationSeconds=0.1
     collapsedImage {
      bIsDynamicallyLoaded=false
      drawAs=Image
      imageType=Vector
      mirroring=NoMirror
      resourceName=../../../Engine/Content/Slate/Starship/Common/caret-right.svg
      resourceObject=null
      tiling=NoTile
      imageSize=<16 16>
      margin=<ltrb 0 0 0 0>
      outlineSettings {
       bUseBrushTransparency=false
       roundingType=HalfHeightRadius
       width=0
       color {colorUseRule=UseColor_Specified, specifiedColor=<rgba 0 0 0 0>}
       cornerRadii=<0 0 0 0>
      }
      tintColor {colorUseRule=UseColor_Specified, specifiedColor=<rgba 0.5271 0.5271 0.5271 1>}
      uVRegion {bIsValid=false, max=<0 0>, min=<0 0>}
     }
     expandedImage {
      bIsDynamicallyLoaded=false
      drawAs=Image
      imageType=Vector
      mirroring=NoMirror
      resourceName=../../../Engine/Content/Slate/Starship/Common/caret-down.svg
      resourceObject=null
      tiling=NoTile
      imageSize=<16 16>
      margin=<ltrb 0 0 0 0>
      outlineSettings {
       bUseBrushTransparency=false
       roundingType=HalfHeightRadius
       width=0
       color {colorUseRule=UseColor_Specified, specifiedColor=<rgba 0 0 0 0>}
       cornerRadii=<0 0 0 0>
      }
      tintColor {colorUseRule=UseColor_Specified, specifiedColor=<rgba 0.5271 0.5271 0.5271 1>}
      uVRegion {bIsValid=false, max=<0 0>, min=<0 0>}
     }
    }
   }
   slot {class=ScrollBoxSlot}
  }
  TextBlock_66 {
   widgetClassPath=&TextBlock
   namedSlotHost=ExpandableArea_52
   widget=TextBlock_66
   properties {
    text=Post Recording
    font {
     bForceMonospaced=false
     bMaterialIsStencil=false
     fontMaterial=null
     fontObject=&Roboto
     letterSpacing=0
     monospacedWidth=1
     size=20
     skewAmount=0
     typefaceFontName=Bold
     outlineSettings {
      bApplyOutlineToDropShadows=false
      bMiteredCorners=false
      bSeparateFillAlpha=false
      outlineMaterial=null
      outlineSize=0
      outlineColor=<rgba 0 0 0 1>
     }
    }
   }
  }
  HorizontalBox_26 {widgetClassPath=&HorizontalBox, namedSlotHost=ExpandableArea_52, widget=HorizontalBox_26}
  VerticalBox_53 {
   widgetClassPath=&VerticalBox
   parent=HorizontalBox_26
   widget=VerticalBox_53
   slot {class=HorizontalBoxSlot, properties {padding=<ltrb 10 10 10 10>, size {sizeRule=Fill, value=1}}}
  }
  DetailsView_PostProcess {
   widgetClassPath=&DetailsView
   parent=VerticalBox_53
   widget=DetailsView_PostProcess
   properties {categoriesToShow [Post Process Settings]}
   slot {class=VerticalBoxSlot}
  }
  HorizontalBox_68 {widgetClassPath=&HorizontalBox, parent=VerticalBox_53, widget=HorizontalBox_68, slot {class=VerticalBoxSlot}}
  BTN_ConverPlayerToMannequin {
   widgetClassPath=&Button
   parent=HorizontalBox_68
   widget=BTN_ConverPlayerToMannequin
   properties {widgetStyle {
     clickedSlateSound {resourceObject=null}
     disabled {
      bIsDynamicallyLoaded=false
      drawAs=RoundedBox
      imageType=NoImage
      mirroring=NoMirror
      resourceName=None
      resourceObject=null
      tiling=NoTile
      imageSize=<0 0>
      margin=<ltrb 0 0 0 0>
      outlineSettings {
       bUseBrushTransparency=false
       roundingType=FixedRadius
       width=1
       color {colorUseRule=UseColor_ColorTable, specifiedColor=<rgba 0 0 0 0>}
       cornerRadii=<4 4 4 4>
      }
      tintColor {colorUseRule=UseColor_Specified, specifiedColor=<rgba 0.0395 0.0395 0.0395 1>}
      uVRegion {bIsValid=false, max=<0 0>, min=<0 0>}
     }
     disabledForeground {colorUseRule=UseColor_ColorTable, specifiedColor=<rgba 0 0 0 0>}
     hovered {
      bIsDynamicallyLoaded=false
      drawAs=RoundedBox
      imageType=NoImage
      mirroring=NoMirror
      resourceName=None
      resourceObject=null
      tiling=NoTile
      imageSize=<32 32>
      margin=<ltrb 0 0 0 0>
      outlineSettings {
       bUseBrushTransparency=false
       roundingType=FixedRadius
       width=1
       color {colorUseRule=UseColor_ColorTable, specifiedColor=<rgba 0 0 0 0>}
       cornerRadii=<4 4 4 4>
      }
      tintColor {colorUseRule=UseColor_Specified, specifiedColor=<rgba 0.0953 0.0953 0.0953 1>}
      uVRegion {bIsValid=false, max=<0 0>, min=<0 0>}
     }
     hoveredForeground {colorUseRule=UseColor_ColorTable, specifiedColor=<rgba 0 0 0 0>}
     hoveredSlateSound {resourceObject=null}
     normal {
      bIsDynamicallyLoaded=false
      drawAs=RoundedBox
      imageType=NoImage
      mirroring=NoMirror
      resourceName=None
      resourceObject=null
      tiling=NoTile
      imageSize=<32 32>
      margin=<ltrb 0 0 0 0>
      outlineSettings {
       bUseBrushTransparency=false
       roundingType=FixedRadius
       width=1
       color {colorUseRule=UseColor_ColorTable, specifiedColor=<rgba 0 0 0 0>}
       cornerRadii=<4 4 4 4>
      }
      tintColor {colorUseRule=UseColor_Specified, specifiedColor=<rgba 0.0395 0.0395 0.0395 1>}
      uVRegion {bIsValid=false, max=<0 0>, min=<0 0>}
     }
     normalForeground {colorUseRule=UseColor_ColorTable, specifiedColor=<rgba 0 0 0 0>}
     normalPadding=<ltrb 12 1.5 12 1.5>
     pressed {
      bIsDynamicallyLoaded=false
      drawAs=RoundedBox
      imageType=NoImage
      mirroring=NoMirror
      resourceName=None
      resourceObject=null
      tiling=NoTile
      imageSize=<32 32>
      margin=<ltrb 0 0 0 0>
      outlineSettings {
       bUseBrushTransparency=false
       roundingType=FixedRadius
       width=1
       color {colorUseRule=UseColor_ColorTable, specifiedColor=<rgba 0 0 0 0>}
       cornerRadii=<4 4 4 4>
      }
      tintColor {colorUseRule=UseColor_Specified, specifiedColor=<rgba 0.0284 0.0284 0.0284 1>}
      uVRegion {bIsValid=false, max=<0 0>, min=<0 0>}
     }
     pressedForeground {colorUseRule=UseColor_ColorTable, specifiedColor=<rgba 0 0 0 0>}
     pressedPadding=<ltrb 12 2.5 12 0.5>
     pressedSlateSound {resourceObject=null}
    }}
   slot {class=HorizontalBoxSlot, properties {padding=<ltrb 10 10 10 10>, size {sizeRule=Fill, value=1}}}
  }
  TextBlock_1 {
   widgetClassPath=&TextBlock
   parent=BTN_ConverPlayerToMannequin
   widget=TextBlock_1
   properties {
    text=Convert Player to Mannequin
    font {
     bForceMonospaced=false
     bMaterialIsStencil=false
     fontMaterial=null
     fontObject=&Roboto
     letterSpacing=0
     monospacedWidth=1
     size=12
     skewAmount=0
     typefaceFontName=Regular
     outlineSettings {
      bApplyOutlineToDropShadows=false
      bMiteredCorners=false
      bSeparateFillAlpha=false
      outlineMaterial=null
      outlineSize=0
      outlineColor=<rgba 0 0 0 1>
     }
    }
   }
   slot {class=ButtonSlot}
  }
  HorizontalBox_2 {widgetClassPath=&HorizontalBox, parent=VerticalBox_53, widget=HorizontalBox_2, slot {class=VerticalBoxSlot}}
  VerticalBox {
   widgetClassPath=&VerticalBox
   parent=HorizontalBox_26
   widget=VerticalBox
   slot {class=HorizontalBoxSlot, properties {padding=<ltrb 10 10 10 10>, size {sizeRule=Fill, value=1}}}
  }
  DetailsView_AddRecordedMannequinTake {
   widgetClassPath=&DetailsView
   parent=VerticalBox
   widget=DetailsView_AddRecordedMannequinTake
   properties {categoriesToShow [Add Recorded Take Settings]}
   slot {class=VerticalBoxSlot}
  }
  HorizontalBox_1 {widgetClassPath=&HorizontalBox, parent=VerticalBox, widget=HorizontalBox_1, slot {class=VerticalBoxSlot}}
  BTN_AddRecordedFNMannequinTake {
   widgetClassPath=&Button
   parent=HorizontalBox_1
   widget=BTN_AddRecordedFNMannequinTake
   properties {widgetStyle {
     clickedSlateSound {resourceObject=null}
     disabled {
      bIsDynamicallyLoaded=false
      drawAs=RoundedBox
      imageType=NoImage
      mirroring=NoMirror
      resourceName=None
      resourceObject=null
      tiling=NoTile
      imageSize=<0 0>
      margin=<ltrb 0 0 0 0>
      outlineSettings {
       bUseBrushTransparency=false
       roundingType=FixedRadius
       width=1
       color {colorUseRule=UseColor_ColorTable, specifiedColor=<rgba 0 0 0 0>}
       cornerRadii=<4 4 4 4>
      }
      tintColor {colorUseRule=UseColor_Specified, specifiedColor=<rgba 0.0395 0.0395 0.0395 1>}
      uVRegion {bIsValid=false, max=<0 0>, min=<0 0>}
     }
     disabledForeground {colorUseRule=UseColor_ColorTable, specifiedColor=<rgba 0 0 0 0>}
     hovered {
      bIsDynamicallyLoaded=false
      drawAs=RoundedBox
      imageType=NoImage
      mirroring=NoMirror
      resourceName=None
      resourceObject=null
      tiling=NoTile
      imageSize=<32 32>
      margin=<ltrb 0 0 0 0>
      outlineSettings {
       bUseBrushTransparency=false
       roundingType=FixedRadius
       width=1
       color {colorUseRule=UseColor_ColorTable, specifiedColor=<rgba 0 0 0 0>}
       cornerRadii=<4 4 4 4>
      }
      tintColor {colorUseRule=UseColor_Specified, specifiedColor=<rgba 0.0953 0.0953 0.0953 1>}
      uVRegion {bIsValid=false, max=<0 0>, min=<0 0>}
     }
     hoveredForeground {colorUseRule=UseColor_ColorTable, specifiedColor=<rgba 0 0 0 0>}
     hoveredSlateSound {resourceObject=null}
     normal {
      bIsDynamicallyLoaded=false
      drawAs=RoundedBox
      imageType=NoImage
      mirroring=NoMirror
      resourceName=None
      resourceObject=null
      tiling=NoTile
      imageSize=<32 32>
      margin=<ltrb 0 0 0 0>
      outlineSettings {
       bUseBrushTransparency=false
       roundingType=FixedRadius
       width=1
       color {colorUseRule=UseColor_ColorTable, specifiedColor=<rgba 0 0 0 0>}
       cornerRadii=<4 4 4 4>
      }
      tintColor {colorUseRule=UseColor_Specified, specifiedColor=<rgba 0.0395 0.0395 0.0395 1>}
      uVRegion {bIsValid=false, max=<0 0>, min=<0 0>}
     }
     normalForeground {colorUseRule=UseColor_ColorTable, specifiedColor=<rgba 0 0 0 0>}
     normalPadding=<ltrb 12 1.5 12 1.5>
     pressed {
      bIsDynamicallyLoaded=false
      drawAs=RoundedBox
      imageType=NoImage
      mirroring=NoMirror
      resourceName=None
      resourceObject=null
      tiling=NoTile
      imageSize=<32 32>
      margin=<ltrb 0 0 0 0>
      outlineSettings {
       bUseBrushTransparency=false
       roundingType=FixedRadius
       width=1
       color {colorUseRule=UseColor_ColorTable, specifiedColor=<rgba 0 0 0 0>}
       cornerRadii=<4 4 4 4>
      }
      tintColor {colorUseRule=UseColor_Specified, specifiedColor=<rgba 0.0284 0.0284 0.0284 1>}
      uVRegion {bIsValid=false, max=<0 0>, min=<0 0>}
     }
     pressedForeground {colorUseRule=UseColor_ColorTable, specifiedColor=<rgba 0 0 0 0>}
     pressedPadding=<ltrb 12 2.5 12 0.5>
     pressedSlateSound {resourceObject=null}
    }}
   slot {class=HorizontalBoxSlot, properties {padding=<ltrb 10 10 10 10>, size {sizeRule=Fill, value=1}}}
  }
  TextBlock_5 {
   widgetClassPath=&TextBlock
   parent=BTN_AddRecordedFNMannequinTake
   widget=TextBlock_5
   properties {
    text=Add Mannequin Take
    colorAndOpacity {colorUseRule=UseColor_Specified, specifiedColor=<rgba 1 0 0 1>}
    font {
     bForceMonospaced=false
     bMaterialIsStencil=false
     fontMaterial=null
     fontObject=&Roboto
     letterSpacing=0
     monospacedWidth=1
     size=12
     skewAmount=0
     typefaceFontName=Regular
     outlineSettings {
      bApplyOutlineToDropShadows=false
      bMiteredCorners=false
      bSeparateFillAlpha=false
      outlineMaterial=null
      outlineSize=0
      outlineColor=<rgba 0 0 0 1>
     }
    }
   }
   slot {class=ButtonSlot}
  }
 ]
}
